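# Niagara system for the male Space Wanderer glider's thrusters, with mirrored left/right flame meshes, SubUV fire sprites, front flares, glow decals and a light.
# Emitters pin particles to bones or sockets on User.SkeletalMesh.
# System acceleration ramps drive spawn rate, color, size and lifetime, while User.ForwardDot adjusts flare material parameters.
# Main flames use spawn groups to zero either blue or purple scale across two mesh renderers.
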NiagaraSystem /BRCosmetics/Gadgets/Assets/VinderTech_GliderChute/Glider_Space_Wanderer_Male/VFX/NS_Glider_SpaceWanderer_Male_Thruster
dependencies {
 usedDataInterfaces [&NiagaraDataInterfaceCurve, &NiagaraDataInterfaceVectorCurve]
 usedDynamicInputs []
 usedModules [
  &self:SystemSpawnScript.SystemScriptSource.SystemScriptGraph.NiagaraNodeAssignment_1.SetVariables_#1
  &SystemState
  &self:SystemSpawnScript.SystemScriptSource.SystemScriptGraph.NiagaraNodeAssignment_0.SetVariables_#2
 ]
}
emitters {
 Blue_Decal_Large {
  emitterUpdateScript [
   /SetVariables(Emitter.AccelerationRamp:NiagaraFloat=$System.AccelerationRamp) {isSetParameters=true, module=SetVariables_#3}
   &EmitterState()
   &SpawnRate(SpawnRate=6, Use Spawn Probability=false)
  ]
  particleSpawnScript [
   &FN_SetSpawnGroup_v2(Number of Spawn Groups=2) {enabled=false}
   &FN_SetSpawnGroup(Number of Spawn Groups=2)
   /Niagara/Modules/Spawn/Initialization/InitializeParticle(
    Color=&MakeLinearColorFromVectorAndFloat(
     "Vector (RGB)"=&MakeVectorFromLinearColorRGB(Color=&LerpLinearColors(
       StartColor=<rgba 0 1.2146 9 1>
       EndColor=<rgba 0 7.2876 18 1>
       LerpFactor=&ClampFloat(Float=$System.AccelerationRamp_02)
      ))
     "Float (Alpha)"=&UniformRangedFloat(Minimum=0.5, Maximum=0.8)
    )
    Lifetime=&UniformRangedFloat(Minimum=2, Maximum=2)
    Module.Write Color=true
    Module.Write Lifetime=true
    Module.Write Mass=false
    Module.Write Material Random=true
    Module.Write Position=false
    Module.Write Scale=false
    Module.Write SpriteRotation=true
    Module.Write SpriteSize=true
    Sprite Rotation=&UniformRangedFloat(Minimum=0, Maximum=359)
    Sprite Size=&Vector2DFromFloat(Value=&Lerp_Float(
      A=&UniformRangedFloat(Minimum=65, Maximum=75)
      B=&UniformRangedFloat(Minimum=65, Maximum=75)
      Alpha=$Emitter.AccelerationRamp
     ))
   )
  ]
  particleUpdateScript [
   &DynamicMaterialParameters(
    Index 0 Param 1=0.8
    Index 0 Param 2=0.1
    Index 0 Param 3=1
    Module.Param0WriteEnabled=true
    Module.Param1WriteEnabled=true
    Module.Param2WriteEnabled=true
    Module.Param3WriteEnabled=false
   )
   &CameraOffset(Camera Offset Amount=&FloatFromCurve(FloatCurve=DataInterface() {propertyValues {Curve {keys [<at 0 30>, {arriveTangent=-47.8205, leaveTangent=-47.8205, time=1, value=60}]}}}))
   &SkeletalMeshLocation(
    Apply Sampled Orientation to Mesh Particles=true
    Bone / Socket Index=$Particles.Initial.SpawnGroup
    Inherit Velocity=false
    MaskSpawnGroup=false
    Mesh Sampling Type=::NewEnumerator5
    Sampled Position Offset=&Lerp_Vector(A=<0 0 20>, B=<0 0 -20>, Alpha=&MakeFloatFromInt(Int=$Particles.Initial.SpawnGroup))
    Sampled Position Offset Coordinate Space:ENiagaraMeshTransforms=::NewEnumerator3
    Skeletal Mesh=$User.SkeletalMesh
    Socket Sampling Mode=::NewEnumerator5
   )
   &ParticleState()
   &ScaleSpriteSize(Scale Factor=&Vector2DFromFloat(Value=&FloatFromCurve(FloatCurve=DataInterface() {propertyValues {Curve {keys [<at 0 2 RCIM_Cubic>, {arriveTangent=-3.004, interpMode=RCIM_Cubic, leaveTangent=-3.004, tangentMode=RCTM_User, time=1, value=1.5}]}}})))
   &ScaleColor(
    Scale Alpha=&FloatFromCurve(FloatCurve=DataInterface() {propertyValues {Curve {keys [
         {arriveTangent=6.0721, interpMode=RCIM_Cubic, leaveTangent=-0.24, tangentMode=RCTM_User, time=0, value=0}
         <at 0.6 1 RCIM_Cubic>
         {arriveTangent=-6.1189, interpMode=RCIM_Cubic, leaveTangent=-6.1189, tangentMode=RCTM_User, time=1, value=0}
        ]}}})
    ScaleA=true
    ScaleRGB=false
   )
  ]
  rendererClasses [&NiagaraSpriteRendererProperties]
  renderers [
   {rendererClass=&NiagaraSpriteRendererProperties}
   {rendererClass=&NiagaraSpriteRendererProperties, rendererIndex=1}
  ]
 }
 Blue_Decal_Small {
  emitterUpdateScript [
   /SetVariables(Emitter.AccelerationRamp:NiagaraFloat=$System.AccelerationRamp) {isSetParameters=true, module=SetVariables_#3}
   &EmitterState()
   &SpawnRate(SpawnRate=8, Use Spawn Probability=false)
  ]
  particleSpawnScript [
   &FN_SetSpawnGroup_v2(Number of Spawn Groups=4) {enabled=false}
   &FN_SetSpawnGroup()
   &InitializeParticle(
    Color=&MakeLinearColorFromVectorAndFloat(
     "Vector (RGB)"=&MakeVectorFromLinearColorRGB(Color=&LerpLinearColors(
       StartColor=<rgba 0 8.8889 25 1>
       EndColor=<rgba 0 26.3164 65 1>
       LerpFactor=&ClampFloat(Float=$System.AccelerationRamp_02)
      ))
     "Float (Alpha)"=&UniformRangedFloat(Minimum=0.5, Maximum=0.8)
    )
    Lifetime=&UniformRangedFloat(Minimum=2, Maximum=2)
    Module.Write Color=true
    Module.Write Lifetime=true
    Module.Write Mass=false
    Module.Write Material Random=true
    Module.Write Position=false
    Module.Write Scale=false
    Module.Write SpriteRotation=true
    Module.Write SpriteSize=true
    Sprite Rotation=&UniformRangedFloat(Minimum=0, Maximum=359)
    Sprite Size=<8 8>
   )
  ]
  particleUpdateScript [
   &DynamicMaterialParameters(
    Index 0 Param 1=0.8
    Index 0 Param 2=0.1
    Index 0 Param 3=1
    Module.Param0WriteEnabled=true
    Module.Param1WriteEnabled=true
    Module.Param2WriteEnabled=true
    Module.Param3WriteEnabled=false
   )
   &CameraOffset(Camera Offset Amount=&FloatFromCurve(FloatCurve=DataInterface() {propertyValues {Curve {keys [<at 0 5>, {arriveTangent=-47.8205, leaveTangent=-47.8205, time=1, value=9}]}}}))
   &SkeletalMeshLocation(
    Apply Sampled Orientation to Mesh Particles=true
    Bone / Socket Index=&Add_Integer(A=8, B=$Particles.Initial.SpawnGroup)
    Inherit Velocity=false
    MaskSpawnGroup=false
    Mesh Sampling Type=::NewEnumerator5
    Sampled Position Offset Coordinate Space:ENiagaraMeshTransforms=::NewEnumerator3
    Skeletal Mesh=$User.SkeletalMesh
    Socket Sampling Mode=::NewEnumerator5
   )
   &ParticleState()
   &ScaleSpriteSize(Scale Factor=&Vector2DFromFloat(Value=&FloatFromCurve(FloatCurve=DataInterface() {propertyValues {Curve {keys [<at 0 2 RCIM_Cubic>, {arriveTangent=-3.004, interpMode=RCIM_Cubic, leaveTangent=-3.004, tangentMode=RCTM_User, time=1, value=1.5}]}}})))
   &ScaleColor(
    Scale Alpha=&FloatFromCurve(FloatCurve=DataInterface() {propertyValues {Curve {keys [
         {arriveTangent=6.0721, interpMode=RCIM_Cubic, leaveTangent=-0.24, tangentMode=RCTM_User, time=0, value=0}
         <at 0.6 1 RCIM_Cubic>
         {arriveTangent=-6.1189, interpMode=RCIM_Cubic, leaveTangent=-6.1189, tangentMode=RCTM_User, time=1, value=0}
        ]}}})
    ScaleA=true
    ScaleRGB=false
   )
  ]
  rendererClasses [&NiagaraSpriteRendererProperties]
  renderers [
   {rendererClass=&NiagaraSpriteRendererProperties}
   {rendererClass=&NiagaraSpriteRendererProperties, rendererIndex=1}
  ]
 }
 Fire_L {
  emitterUpdateScript [
   /SetVariables(Emitter.AccelerationRamp:NiagaraFloat=$System.AccelerationRamp) {isSetParameters=true, module=SetVariables_#3}
   &EmitterState()
   &SpawnRate(
    SpawnRate=&Lerp_Float(A=15, B=25, Alpha=&ClampFloat(Float=$System.AccelerationRamp_02))
    Use Spawn Probability=false
   )
  ]
  particleSpawnScript [
   &InitializeParticle(
    Color=&LerpLinearColors(
     StartColor=&LerpLinearColors(
      StartColor=<rgba 0 0 50 0.8>
      EndColor=<rgba 0 2.4292 3 0.8>
      LerpFactor=&UniformRangedFloat(Minimum=0, Maximum=1)
     )
     EndColor=<rgba 0 8.0974 10 0.8>
     LerpFactor=$Emitter.AccelerationRamp
    )
    Lifetime=&Lerp_Float(
     A=&UniformRangedFloat(Minimum=1.25, Maximum=1.8)
     B=&UniformRangedFloat(Minimum=0.8, Maximum=1.25)
     Alpha=$Emitter.AccelerationRamp
    )
    Module.Write Color=true
    Module.Write Lifetime=true
    Module.Write Mass=false
    Module.Write Material Random=true
    Module.Write Position=false
    Module.Write Scale=false
    Module.Write SpriteRotation=true
    Module.Write SpriteSize=true
    Position=&SimulationPosition()
    Sprite Rotation=&UniformRangedFloat(Minimum=-180, Maximum=180)
    Sprite Size=&Vector2DFromFloat(Value=&Lerp_Float(
      A=&UniformRangedFloat(Minimum=20, Maximum=45)
      B=&UniformRangedFloat(Minimum=30, Maximum=50)
      Alpha=&ClampFloat(Float=$System.AccelerationRamp_02)
     ))
   )
   /Niagara/Modules/Spawn/Location/SphereLocation(Hemisphere Z=true, MaskSpawnGroup=false, Module.Override Randomness=false, Non Uniform Scale=<1 1 -1>, Sphere Radius=1, Surface Only=true)
   /SetVariables(Particles.SphereVector:Vector3f=&Multiply_Vector(
     A=$Output.SphereLocation.SphereNormal
     B=&Lerp_Vector(
      A=&MakeVector(
       X=&UniformRangedFloat(Minimum=15, Maximum=25)
       Y=&UniformRangedFloat(Minimum=15, Maximum=25)
       Z=&UniformRangedFloat(Minimum=10, Maximum=55)
      )
      B=&MakeVector(
       X=&UniformRangedFloat(Minimum=12, Maximum=18)
       Y=&UniformRangedFloat(Minimum=12, Maximum=18)
       Z=&UniformRangedFloat(Minimum=9, Maximum=45)
      )
      Alpha=&ClampFloat(Float=$System.AccelerationRamp_02)
     )
    )) {isSetParameters=true}
   &DynamicMaterialParameters(
    Index 0 Param 2=&UniformRangedFloat(Minimum=0, Maximum=1)
    Index 0 Param 3=1
    Index 0 Param 4=&UniformRangedFloat(Minimum=1, Maximum=2)
    Module.Param0WriteEnabled=false
    Module.Param1WriteEnabled=true
    Module.Param2WriteEnabled=false
    Module.Param3WriteEnabled=true
   ) {module=DynamicMaterialParameters002}
   /SetVariables(Particles.FireMovementRange:NiagaraFloat=&UniformRangedFloat(Minimum=-5, Maximum=35)) {isSetParameters=true, module=SetVariables_#4}
  ]
  particleUpdateScript [
   &CameraOffset(Camera Offset Amount=&FloatFromCurve(FloatCurve=DataInterface() {propertyValues {Curve {keys [<at 0 0>,<at 1 52.5>]}}}))
   &ParticleState()
   &SkeletalMeshLocation(
    Apply Sampled Orientation to Mesh Particles=true
    Inherit Velocity=false
    Interpolate Mesh Sampling=false
    MaskSpawnGroup=false
    Mesh Sampling Type=::NewEnumerator5
    Sampled Position Offset=&Lerp_Vector(
     A=&Add_Vector(A=<0 0 75>, B=$Particles.SphereVector)
     B=&Add_Vector(A=&MakeVector(X=0, Y=0, Z=$Particles.Initial.FireMovementRange), B=$Particles.SphereVector)
     Alpha=&FloatFromCurve(FloatCurve=DataInterface() {propertyValues {Curve {keys [<at 0 0>,<at 1 1>]}}})
    )
    Sampled Position Offset Coordinate Space:ENiagaraMeshTransforms=::NewEnumerator3
    Skeletal Mesh=$User.SkeletalMesh
    Socket Sampling Mode=::NewEnumerator5
   )
   /Niagara/Modules/Update/SubUV/SubUVAnimation(Number Of Frames=64)
   &ScaleSpriteSize(Scale Factor=&Vector2DFromFloat(Value=&FloatFromCurve(FloatCurve=DataInterface() {propertyValues {Curve {keys [
          {arriveTangent=7.5764, interpMode=RCIM_Cubic, time=0, value=0.02}
          {arriveTangent=1.2063, interpMode=RCIM_Cubic, leaveTangent=1.2063, tangentMode=RCTM_User, time=0.3, value=0.99}
          <at 1 0.5 RCIM_Cubic>
         ]}}})))
   &ScaleColor(
    Scale Alpha=&FloatFromCurve(FloatCurve=DataInterface() {propertyValues {Curve {keys [
         {arriveTangent=8.7245, interpMode=RCIM_Cubic, leaveTangent=8.7245, tangentMode=RCTM_User, time=0, value=0}
         <at 0.2 1 RCIM_Cubic>
         {arriveTangent=-0.5281, interpMode=RCIM_Cubic, leaveTangent=-0.5281, tangentMode=RCTM_User, time=1, value=0}
        ]}}})
    ScaleA=true
    ScaleRGB=true
   )
   &DynamicMaterialParameters(
    Index 0 Param 1=&FloatFromCurve(FloatCurve=DataInterface() {propertyValues {Curve {keys [{arriveTangent=2.7551, interpMode=RCIM_Cubic, time=0, value=0.7}, <at 0.7 1>]}}})
    Index 0 Param 3=1.5
    Index 0 Param 4=1
    Module.Param0WriteEnabled=true
    Module.Param1WriteEnabled=true
    Module.Param2WriteEnabled=true
    Module.Param3WriteEnabled=false
   ) {module=DynamicMaterialParameters001}
  ]
  rendererClasses [&NiagaraSpriteRendererProperties]
  renderers [{rendererClass=&NiagaraSpriteRendererProperties}]
 }
 Fire_R {
  emitterUpdateScript [
   /SetVariables(Emitter.AccelerationRamp:NiagaraFloat=$System.AccelerationRamp) {isSetParameters=true, module=SetVariables_#3}
   &EmitterState()
   &SpawnRate(
    SpawnRate=&Lerp_Float(A=15, B=25, Alpha=&ClampFloat(Float=$System.AccelerationRamp_02))
    Use Spawn Probability=false
   )
  ]
  particleSpawnScript [
   &InitializeParticle(
    Color=&LerpLinearColors(
     StartColor=&LerpLinearColors(
      StartColor=<rgba 0 0 50 0.8>
      EndColor=<rgba 0 2.4292 3 0.8>
      LerpFactor=&UniformRangedFloat(Minimum=0, Maximum=1)
     )
     EndColor=<rgba 0 8.0974 10 0.8>
     LerpFactor=$Emitter.AccelerationRamp
    )
    Lifetime=&Lerp_Float(
     A=&UniformRangedFloat(Minimum=1.25, Maximum=1.8)
     B=&UniformRangedFloat(Minimum=0.8, Maximum=1.25)
     Alpha=$Emitter.AccelerationRamp
    )
    Module.Write Color=true
    Module.Write Lifetime=true
    Module.Write Mass=false
    Module.Write Material Random=true
    Module.Write Position=false
    Module.Write Scale=false
    Module.Write SpriteRotation=true
    Module.Write SpriteSize=true
    Position=&SimulationPosition()
    Sprite Rotation=&UniformRangedFloat(Minimum=-180, Maximum=180)
    Sprite Size=&Vector2DFromFloat(Value=&Lerp_Float(
      A=&UniformRangedFloat(Minimum=20, Maximum=45)
      B=&UniformRangedFloat(Minimum=30, Maximum=50)
      Alpha=&ClampFloat(Float=$System.AccelerationRamp_02)
     ))
   )
   &SphereLocation(Hemisphere Z=true, MaskSpawnGroup=false, Module.Override Randomness=false, Sphere Radius=1, Surface Only=true)
   /SetVariables(Particles.SphereVector:Vector3f=&Multiply_Vector(
     A=$Output.SphereLocation.SphereNormal
     B=&Lerp_Vector(
      A=&MakeVector(
       X=&UniformRangedFloat(Minimum=15, Maximum=25)
       Y=&UniformRangedFloat(Minimum=15, Maximum=25)
       Z=&UniformRangedFloat(Minimum=10, Maximum=55)
      )
      B=&MakeVector(
       X=&UniformRangedFloat(Minimum=12, Maximum=18)
       Y=&UniformRangedFloat(Minimum=12, Maximum=18)
       Z=&UniformRangedFloat(Minimum=9, Maximum=45)
      )
      Alpha=&ClampFloat(Float=$System.AccelerationRamp_02)
     )
    )) {isSetParameters=true}
   &DynamicMaterialParameters(
    Index 0 Param 2=&UniformRangedFloat(Minimum=0, Maximum=1)
    Index 0 Param 3=1
    Index 0 Param 4=&UniformRangedFloat(Minimum=1, Maximum=2)
    Module.Param0WriteEnabled=false
    Module.Param1WriteEnabled=true
    Module.Param2WriteEnabled=false
    Module.Param3WriteEnabled=true
   ) {module=DynamicMaterialParameters002}
   /SetVariables(Particles.FireMovementRange:NiagaraFloat=&UniformRangedFloat(Minimum=5, Maximum=-35)) {isSetParameters=true, module=SetVariables_#4}
  ]
  particleUpdateScript [
   &CameraOffset(Camera Offset Amount=&FloatFromCurve(FloatCurve=DataInterface() {propertyValues {Curve {keys [<at 0 0>,<at 1 52.5>]}}}))
   &ParticleState()
   &SkeletalMeshLocation(
    Apply Sampled Orientation to Mesh Particles=true
    Bone / Socket Index=1
    Inherit Velocity=false
    Interpolate Mesh Sampling=false
    MaskSpawnGroup=false
    Mesh Sampling Type=::NewEnumerator5
    Sampled Position Offset=&Lerp_Vector(
     A=&Add_Vector(A=<0 0 -75>, B=$Particles.SphereVector)
     B=&Add_Vector(A=&MakeVector(X=0, Y=0, Z=$Particles.Initial.FireMovementRange), B=$Particles.SphereVector)
     Alpha=&FloatFromCurve(FloatCurve=DataInterface() {propertyValues {Curve {keys [<at 0 0>,<at 1 1>]}}})
    )
    Sampled Position Offset Coordinate Space:ENiagaraMeshTransforms=::NewEnumerator3
    Skeletal Mesh=$User.SkeletalMesh
    Socket Sampling Mode=::NewEnumerator5
   )
   &SubUVAnimation(Number Of Frames=64)
   &ScaleSpriteSize(Scale Factor=&Vector2DFromFloat(Value=&FloatFromCurve(FloatCurve=DataInterface() {propertyValues {Curve {keys [
          {arriveTangent=7.5764, interpMode=RCIM_Cubic, time=0, value=0.02}
          {arriveTangent=1.2063, interpMode=RCIM_Cubic, leaveTangent=1.2063, tangentMode=RCTM_User, time=0.3, value=0.99}
          <at 1 0.5 RCIM_Cubic>
         ]}}})))
   &ScaleColor(
    Scale Alpha=&FloatFromCurve(FloatCurve=DataInterface() {propertyValues {Curve {keys [
         {arriveTangent=8.7245, interpMode=RCIM_Cubic, leaveTangent=8.7245, tangentMode=RCTM_User, time=0, value=0}
         <at 0.2 1 RCIM_Cubic>
         {arriveTangent=-0.5281, interpMode=RCIM_Cubic, leaveTangent=-0.5281, tangentMode=RCTM_User, time=1, value=0}
        ]}}})
    ScaleA=true
    ScaleRGB=true
   )
   &DynamicMaterialParameters(
    Index 0 Param 1=&FloatFromCurve(FloatCurve=DataInterface() {propertyValues {Curve {keys [{arriveTangent=2.7551, interpMode=RCIM_Cubic, time=0, value=0.7}, <at 0.7 1>]}}})
    Index 0 Param 3=1.5
    Index 0 Param 4=1
    Module.Param0WriteEnabled=true
    Module.Param1WriteEnabled=true
    Module.Param2WriteEnabled=true
    Module.Param3WriteEnabled=false
   ) {module=DynamicMaterialParameters001}
  ]
  rendererClasses [&NiagaraSpriteRendererProperties]
  renderers [{rendererClass=&NiagaraSpriteRendererProperties}]
 }
 Front_Flare_Large_L {
  emitterUpdateScript [
   /Game/Effects/Niagara/Modules/Math/FN_Integrator/FN_Integrator_v2(
    Clamp=true
    DecayRate=HlslExpression() {hlslExpression="User.ForwardDot < .1 ? Engine.DeltaTime : 0"}
    Input=HlslExpression() {hlslExpression="User.ForwardDot > .1 ? Engine.DeltaTime : 0"}
    Max=2
    Namespace:ENiagaraNameSpace=::NewEnumerator1
   )
   /SetVariables(Emitter.AccelerationRamp:NiagaraFloat=&FN_FrontEndFloatModifier(
     Gameplay=/Game/Effects/Niagara/DynamicInputs/Math/FN_Fit(
      Clamp:NiagaraBool=true
      Value:NiagaraFloat=$Emitter.FN_Integrator_v2.Attr
      SourceMin:NiagaraFloat=1
      SourceMax:NiagaraFloat=2
      DestMin:NiagaraFloat=0.25
      DestMax:NiagaraFloat=0.6
     )
     Front End=1
    )) {isSetParameters=true, module=SetVariables_#3}
   &EmitterState(
    Inactive Response=::NewEnumerator1
    Life Cycle Mode=::NewEnumerator1
    Loop Behavior=::NewEnumerator1
    Loop Delay=&FN_FrontEndFloatModifier(Gameplay=1, Front End=0.1)
    Loop Duration Mode=::NewEnumerator1
    UseLoopDelay=false
   )
   &SpawnBurst_Instantaneous(Spawn Count=1, Use Spawn Probability=false)
  ]
  particleSpawnScript [&InitializeParticle(
    Color=<rgba 1.414 0 4 0.5>
    Lifetime=0.6
    Mesh Scale=&VectorFromFloat(Value=&FN_FrontEndFloatModifier(Gameplay=1, Front End=0.7))
    Module.Write Color=false
    Module.Write Lifetime=false
    Module.Write Mass=false
    Module.Write Material Random=true
    Module.Write Position=false
    Module.Write Scale=true
    Module.Write SpriteRotation=false
    Module.Write SpriteSize=false
    Sprite Size=<5 5>
   )]
  particleUpdateScript [
   &ParticleState(Kill Particles When Lifetime Has Elapsed=false, Let Infinitely Lived Particles Die When Emitter Deactivates=true)
   &SkeletalMeshLocation(
    Apply Sampled Orientation to Mesh Particles=true
    Bone / Socket Index=2
    Inherit Velocity=false
    MaskSpawnGroup=false
    Mesh Sampling Type=::NewEnumerator5
    Sampled Position Offset Coordinate Space:ENiagaraMeshTransforms=::NewEnumerator3
    Skeletal Mesh=$User.SkeletalMesh
    Socket Sampling Mode=::NewEnumerator5
   )
   &Color(Color=<rgba 6 0 12 0.8>)
   &ScaleColor(
    Color Value To Scale=$Particles.Color
    Scale Alpha=&FloatFromCurve(
     FloatCurve=DataInterface() {propertyValues {Curve {keys [<at 0 0.1>,<at 1 0.5>]}}}
     CurveIndex=&ClampFloat(Float=$System.AccelerationRamp_02)
    )
    ScaleA=true
    ScaleRGB=false
   )
   /SetVariables(Particles.MeshOrientation=&MultiplyQuaternion(
     Quaternion A=$Particles.MeshOrientation
     Quaternion B=&MakeQuatFromAxisAngle(Axis=<0 1 0>, AngleInDegrees=-90)
    )) {isSetParameters=true, module=SetVariables_#5}
   /SetVariables(Particles.MeshOrientation=&MultiplyQuaternion(
     Quaternion A=$Particles.MeshOrientation
     Quaternion B=&MakeQuatFromAxisAngle(Axis=<1 0 0>, AngleInDegrees=180)
    )) {isSetParameters=true}
   &DynamicMaterialParameters(
    Index 0 Param 1=&FloatFromCurve(
     FloatCurve=DataInterface() {propertyValues {Curve {keys [<at 0 2>,<at 1 1.5>]}}}
     CurveIndex=&Ceil_Float(Float=$User.ForwardDot)
    )
    Index 0 Param 2=&FN_FrontEndFloatModifier(
     Gameplay=&FloatFromCurve(
      FloatCurve=DataInterface() {propertyValues {Curve {keys [<at 0 0.6>,<at 1 1.3>]}}}
      CurveIndex=&Ceil_Float(Float=$User.ForwardDot)
     )
     Front End=1.31
    )
    Module.Param0WriteEnabled=true
    Module.Param1WriteEnabled=true
    Module.Param2WriteEnabled=false
    Module.Param3WriteEnabled=false
   )
  ]
  rendererClasses [&NiagaraMeshRendererProperties]
  renderers [{rendererClass=&NiagaraMeshRendererProperties, rendererIndex=0}]
 }
 Front_Flare_Large_R {
  emitterUpdateScript [
   &FN_Integrator_v2(
    Clamp=true
    DecayRate=HlslExpression() {hlslExpression="User.ForwardDot < .1 ? Engine.DeltaTime : 0"}
    Input=HlslExpression() {hlslExpression="User.ForwardDot > .1 ? Engine.DeltaTime : 0"}
    Max=2
    Namespace:ENiagaraNameSpace=::NewEnumerator1
   )
   /SetVariables(Emitter.AccelerationRamp:NiagaraFloat=&FN_FrontEndFloatModifier(
     Gameplay=&FN_Fit(
      Clamp:NiagaraBool=true
      Value:NiagaraFloat=$Emitter.FN_Integrator_v2.Attr
      SourceMin:NiagaraFloat=1
      SourceMax:NiagaraFloat=2
      DestMin:NiagaraFloat=0.25
      DestMax:NiagaraFloat=0.6
     )
     Front End=1
    )) {isSetParameters=true, module=SetVariables_#3}
   &EmitterState(
    Inactive Response=::NewEnumerator1
    Life Cycle Mode=::NewEnumerator1
    Loop Behavior=::NewEnumerator1
    Loop Delay=&FN_FrontEndFloatModifier(Gameplay=1, Front End=0.1)
    Loop Duration Mode=::NewEnumerator1
    UseLoopDelay=false
   )
   &SpawnBurst_Instantaneous(Spawn Count=1, Use Spawn Probability=false)
  ]
  particleSpawnScript [&InitializeParticle(
    Color=<rgba 1.414 0 4 0.5>
    Lifetime=0.6
    Mesh Scale=&VectorFromFloat(Value=&FN_FrontEndFloatModifier(Gameplay=1, Front End=0.7))
    Module.Write Color=false
    Module.Write Lifetime=false
    Module.Write Mass=false
    Module.Write Material Random=true
    Module.Write Position=false
    Module.Write Scale=true
    Module.Write SpriteRotation=true
    Module.Write SpriteSize=false
    Sprite Rotation=0.25
    Sprite Size=<5 5>
   )]
  particleUpdateScript [
   &ParticleState(Kill Particles When Lifetime Has Elapsed=false, Let Infinitely Lived Particles Die When Emitter Deactivates=true)
   &SkeletalMeshLocation(
    Apply Sampled Orientation to Mesh Particles=true
    Bone / Socket Index=3
    Inherit Velocity=false
    MaskSpawnGroup=false
    Mesh Sampling Type=::NewEnumerator5
    Sampled Position Offset Coordinate Space:ENiagaraMeshTransforms=::NewEnumerator3
    Skeletal Mesh=$User.SkeletalMesh
    Socket Sampling Mode=::NewEnumerator5
   )
   &Color(Color=<rgba 6 0 12 0.8>)
   &ScaleColor(
    Color Value To Scale=$Particles.Color
    Scale Alpha=&FloatFromCurve(
     FloatCurve=DataInterface() {propertyValues {Curve {keys [<at 0 0.1>,<at 1 0.5>]}}}
     CurveIndex=&ClampFloat(Float=$System.AccelerationRamp_02)
    )
    ScaleA=true
    ScaleRGB=false
   )
   /SetVariables(Particles.MeshOrientation=&MultiplyQuaternion(
     Quaternion A=$Particles.MeshOrientation
     Quaternion B=&MakeQuatFromAxisAngle(Axis=<0 1 0>, AngleInDegrees=90)
    )) {isSetParameters=true, module=SetVariables_#5}
   &DynamicMaterialParameters(
    Index 0 Param 1=&FloatFromCurve(
     FloatCurve=DataInterface() {propertyValues {Curve {keys [<at 0 2>,<at 1 1.5>]}}}
     CurveIndex=&Ceil_Float(Float=$User.ForwardDot)
    )
    Index 0 Param 2=&FN_FrontEndFloatModifier(
     Gameplay=&FloatFromCurve(
      FloatCurve=DataInterface() {propertyValues {Curve {keys [<at 0 0.6>,<at 1 1.3>]}}}
      CurveIndex=&Ceil_Float(Float=$User.ForwardDot)
     )
     Front End=1.3
    )
    Module.Param0WriteEnabled=true
    Module.Param1WriteEnabled=true
    Module.Param2WriteEnabled=false
    Module.Param3WriteEnabled=false
   )
  ]
  rendererClasses [&NiagaraMeshRendererProperties]
  renderers [{rendererClass=&NiagaraMeshRendererProperties, rendererIndex=0}]
 }
 Light {
  emitterUpdateScript [
   &EmitterState(Life Cycle Mode=::NewEnumerator1, Loop Behavior=::NewEnumerator1, Loop Duration Mode=::NewEnumerator1, UseLoopDelay=false)
   &SpawnBurst_Instantaneous(Spawn Count=1, Use Spawn Probability=false)
  ]
  particleSpawnScript [&InitializeParticle(
    Color=&LerpLinearColors(
     StartColor=<rgba 0 2.1023 10 1>
     EndColor=<rgba 0 8.4094 40 1>
     LerpFactor=&ClampFloat(Float=1, Max=$System.AccelerationRamp_02)
    )
    Module.Write Color=false
    Module.Write Lifetime=true
    Module.Write Mass=true
    Module.Write Material Random=true
    Module.Write Position=false
    Module.Write Scale=false
    Module.Write SpriteRotation=false
    Module.Write SpriteSize=true
    Position=&SimulationPosition()
    Sprite Size=<5 5>
   )]
  particleUpdateScript [
   &ParticleState(Kill Particles When Lifetime Has Elapsed=false)
   &SolveForcesAndVelocity(Clamp Velocity=false, Limit Acceleration=false)
   &SkeletalMeshLocation(
    Bone Sampling Mode=::NewEnumerator4
    Inherit Velocity=false
    Interpolate Mesh Sampling=false
    MaskSpawnGroup=false
    Sampled Position Offset=<0 50 -65>
    Sampled Position Offset Coordinate Space:ENiagaraMeshTransforms=::NewEnumerator3
    Skeletal Mesh=$User.SkeletalMesh
   )
   &Color(Color=&LerpLinearColors(
     StartColor=<rgba 0 2.3126 11 1>
     EndColor=<rgba 0 9.6708 23 1>
     LerpFactor=&ClampFloat(Float=1, Max=$System.AccelerationRamp_02)
    ))
  ]
  rendererClasses [&NiagaraLightRendererProperties]
  renderers [{rendererClass=&NiagaraLightRendererProperties, rendererIndex=0}]
 }
 Main_L {
  emitterUpdateScript [
   /SetVariables(Emitter.AccelerationRamp:NiagaraFloat=$System.AccelerationRamp) {isSetParameters=true, module=SetVariables_#6}
   &EmitterState(Inactive Response=::NewEnumerator1, Life Cycle Mode=::NewEnumerator1, Loop Duration=1, UseLoopDelay=false)
   &SpawnRate(SpawnRate=&Lerp_Float(A=15, B=20, Alpha=$Emitter.AccelerationRamp), Use Spawn Probability=false)
  ]
  particleSpawnScript [
   &FN_SetSpawnGroup_v2(Number of Spawn Groups=2) {enabled=false}
   &FN_SetSpawnGroup(Number of Spawn Groups=2)
   &InitializeParticle(
    Color=&LerpLinearColors(
     StartColor=&LerpLinearColors(StartColor=<rgba 1 0 1 1>, EndColor=<rgba 12 0 1 3>, LerpFactor=&ClampFloat(Float=$System.AccelerationRamp_02))
     EndColor=&LerpLinearColors(StartColor=<rgba 1 1 1 1.5>, EndColor=<rgba 2.5 2.5 2.5 0.2>, LerpFactor=$Emitter.AccelerationRamp)
     LerpFactor=HlslExpression() {hlslExpression=Particles.SpawnGroup}
    )
    Lifetime=&Lerp_Float(
     A=&Lerp_Float(A=0.9, B=0.6, Alpha=$Emitter.AccelerationRamp)
     B=&Lerp_Float(A=0.5, B=0.6, Alpha=$Emitter.AccelerationRamp)
     Alpha=HlslExpression() {hlslExpression=Particles.SpawnGroup}
    )
    Mesh Scale=&Lerp_Vector(
     A=&FN_FrontEndVector3Modifier(
      Gameplay=&Multiply_Vector(A=<0.3 0.45 0.45>, B=&Lerp_Vector(A=<0.6 0.7 0.7>, B=<0.5 0.85 0.85>, Alpha=$Emitter.AccelerationRamp))
      Front End=&Multiply_Vector(A=<0.3 0.4 0.4>, B=&Lerp_Vector(A=<0.6 0.7 0.7>, B=<0.5 0.85 0.85>, Alpha=$Emitter.AccelerationRamp))
     )
     B=&Multiply_Vector(A=<0.6 1 1>, B=&Lerp_Vector(A=<0.8 1 1>, B=<0.8 0.87 0.87>, Alpha=$Emitter.AccelerationRamp))
     Alpha=HlslExpression() {hlslExpression=Particles.SpawnGroup}
    )
    Module.Write Color=true
    Module.Write Lifetime=true
    Module.Write Mass=false
    Module.Write Material Random=true
    Module.Write Position=false
    Module.Write Scale=true
    Module.Write SpriteRotation=true
    Module.Write SpriteSize=false
    Position=&SimulationPosition()
    Sprite Rotation=&UniformRangedFloat(Minimum=-0.2, Maximum=0.2)
    Sprite Size=<5 5>
   )
  ]
  particleUpdateScript [
   &ParticleState()
   &SkeletalMeshLocation(
    Apply Sampled Orientation to Mesh Particles=true
    Inherit Velocity=false
    Interpolate Mesh Sampling=false
    MaskSpawnGroup=false
    Mesh Sampling Type=::NewEnumerator5
    Sampled Position Offset=&Lerp_Vector(A=<0 0 70>, B=<0 0 75>, Alpha=&MakeFloatFromInt(Int=$Particles.Initial.SpawnGroup))
    Sampled Position Offset Coordinate Space:ENiagaraMeshTransforms=::NewEnumerator3
    Skeletal Mesh=$User.SkeletalMesh
    Socket Sampling Mode=::NewEnumerator5
   )
   &ScaleColor(
    Scale Alpha=&FloatFromCurve(FloatCurve=DataInterface() {propertyValues {Curve {keys [
         <at 0 0 RCIM_Cubic>
         <at 0.3 1>
         {arriveTangent=0.0062, interpMode=RCIM_Cubic, leaveTangent=0.0062, tangentMode=RCTM_User, time=0.5, value=1}
         {arriveTangent=-3.9755, interpMode=RCIM_Cubic, leaveTangent=-3.9755, tangentMode=RCTM_User, time=1, value=0}
        ]}}})
    Scale RGB=<0.75 1 1>
    ScaleA=true
    ScaleRGB=true
   )
   /SetVariables(Particles.BlueScale:Vector3f=&Lerp_Vector(
     A=<0 0 0>
     B=&Multiply_Vector(
      A=&Multiply_Vector(
       A=$Particles.Initial.Scale
       B=&VectorFromCurve(VectorCurve=DataInterface() {propertyValues {
          XCurve {keys [
            {arriveTangent=0.0368, interpMode=RCIM_Cubic, leaveTangent=0.0368, tangentMode=RCTM_User, time=0, value=1.5}
            {arriveTangent=1.3132, interpMode=RCIM_Cubic, leaveTangent=1.3132, tangentMode=RCTM_User, time=1, value=2.2}
           ]}
          YCurve {keys [
            {arriveTangent=0.0463, interpMode=RCIM_Cubic, leaveTangent=0.0463, tangentMode=RCTM_User, time=0, value=0.5}
            {arriveTangent=-0.8757, interpMode=RCIM_Cubic, leaveTangent=-2.7939, tangentMode=RCTM_User, time=0.9977, value=0.1}
           ]}
          ZCurve {keys [<at 0 0.5 RCIM_Cubic>, {arriveTangent=-1.0353, interpMode=RCIM_Cubic, leaveTangent=-2.975, tangentMode=RCTM_User, time=0.9977, value=0.1}]}
         }})
      )
      B=&VectorFromCurve(
       VectorCurve=DataInterface() {propertyValues {
         XCurve {keys [
           <at 0.2506 0.7 RCIM_Cubic>
           {arriveTangent=0.9503, interpMode=RCIM_Cubic, leaveTangent=0.9503, tangentMode=RCTM_User, time=0.9973, value=1.4}
          ]}
         YCurve {keys [<at 0.2506 0.7 RCIM_Cubic>,<at 1 0.7 RCIM_Cubic>]}
         ZCurve {keys [<at 0.2506 0.7 RCIM_Cubic>,<at 1 0.7 RCIM_Cubic>]}
        }}
       CurveIndex=$Emitter.AccelerationRamp
      )
     )
     Alpha=HlslExpression() {hlslExpression="float(Particles.SpawnGroup)"}
    )) {isSetParameters=true, module=SetVariables_#7}
   /SetVariables(Particles.PurpleScale:Vector3f=&Lerp_Vector(
     A=&Multiply_Vector(
      A=&Multiply_Vector(
       A=$Particles.Initial.Scale
       B=&Lerp_Vector(
        A=&VectorFromCurve(VectorCurve=DataInterface() {propertyValues {
           XCurve {keys [
             {arriveTangent=4.2301, interpMode=RCIM_Cubic, time=0, value=3}
             {arriveTangent=0.0169, interpMode=RCIM_Cubic, leaveTangent=0.0169, tangentMode=RCTM_User, time=1, value=4}
            ]}
           YCurve {keys [
             {arriveTangent=0.0463, interpMode=RCIM_Cubic, leaveTangent=0.0463, tangentMode=RCTM_User, time=0, value=0.98}
             {arriveTangent=-1.8909, interpMode=RCIM_Cubic, leaveTangent=-1.8909, tangentMode=RCTM_User, time=1, value=0.5}
            ]}
           ZCurve {keys [<at 0 0.98 RCIM_Cubic>, {arriveTangent=-1.8424, interpMode=RCIM_Cubic, leaveTangent=-1.8424, tangentMode=RCTM_User, time=1, value=0.5}]}
          }})
        B=&VectorFromCurve(VectorCurve=DataInterface() {propertyValues {
           XCurve {keys [
             {arriveTangent=4.2301, interpMode=RCIM_Cubic, leaveTangent=4.2301, tangentMode=RCTM_User, time=0, value=0.2}
             {arriveTangent=0.0169, interpMode=RCIM_Cubic, leaveTangent=0.0169, tangentMode=RCTM_User, time=1, value=6}
            ]}
           YCurve {keys [
             {arriveTangent=0.0463, interpMode=RCIM_Cubic, leaveTangent=0.0463, tangentMode=RCTM_User, time=0, value=0.98}
             {interpMode=RCIM_Cubic, leaveTangent=-1.8909, time=1, value=0.8}
            ]}
           ZCurve {keys [<at 0 0.98 RCIM_Cubic>, {interpMode=RCIM_Cubic, leaveTangent=-1.8424, time=1, value=0.8}]}
          }})
        Alpha=&ClampFloat(Float=$System.AccelerationRamp_02)
       )
      )
      B=<1 1 1>
     )
     B=<0 0 0>
     Alpha=HlslExpression() {hlslExpression="float(Particles.SpawnGroup)"}
    )) {isSetParameters=true, module=SetVariables_#8}
   /SetVariables(Particles.MeshOrientation=&MultiplyQuaternion(Quaternion A=$Particles.MeshOrientation, Quaternion B=<0 -1 0 1>)) {isSetParameters=true, module=SetVariables_#9}
   /SetVariables(Particles.MeshOrientation=&MultiplyQuaternion(
     Quaternion A=$Particles.MeshOrientation
     Quaternion B=&MakeQuatFromAxisAngle(Axis=<1 0 0>, AngleInDegrees=&Multiply_Float(A=$Particles.MaterialRandom, B=360))
    )) {isSetParameters=true, module=SetVariables_#10}
   &CameraOffset()
  ]
  rendererClasses [&NiagaraMeshRendererProperties]
  renderers [
   {rendererClass=&NiagaraMeshRendererProperties, rendererIndex=0}
   {rendererClass=&NiagaraMeshRendererProperties, rendererIndex=1}
  ]
 }
 Main_R {
  emitterUpdateScript [
   /SetVariables(Emitter.AccelerationRamp:NiagaraFloat=$System.AccelerationRamp) {isSetParameters=true, module=SetVariables_#6}
   &EmitterState(Inactive Response=::NewEnumerator1, Life Cycle Mode=::NewEnumerator1, Loop Duration=1, UseLoopDelay=false)
   &SpawnRate(SpawnRate=&Lerp_Float(A=15, B=20, Alpha=$Emitter.AccelerationRamp), Use Spawn Probability=false)
  ]
  particleSpawnScript [
   &FN_SetSpawnGroup_v2(Number of Spawn Groups=2) {enabled=false}
   &FN_SetSpawnGroup(Number of Spawn Groups=2)
   &InitializeParticle(
    Color=&LerpLinearColors(
     StartColor=&LerpLinearColors(StartColor=<rgba 1 0 1 1>, EndColor=<rgba 12 0 1 3>, LerpFactor=&ClampFloat(Float=$System.AccelerationRamp_02))
     EndColor=&LerpLinearColors(StartColor=<rgba 1 1 1 1.5>, EndColor=<rgba 2.5 2.5 2.5 0.2>, LerpFactor=$Emitter.AccelerationRamp)
     LerpFactor=HlslExpression() {hlslExpression=Particles.SpawnGroup}
    )
    Lifetime=&Lerp_Float(
     A=&Lerp_Float(A=0.9, B=0.6, Alpha=$Emitter.AccelerationRamp)
     B=&Lerp_Float(A=0.5, B=0.6, Alpha=$Emitter.AccelerationRamp)
     Alpha=HlslExpression() {hlslExpression=Particles.SpawnGroup}
    )
    Mesh Scale=&Lerp_Vector(
     A=&FN_FrontEndVector3Modifier(
      Gameplay=&Multiply_Vector(A=<0.3 0.45 0.45>, B=&Lerp_Vector(A=<0.6 0.7 0.7>, B=<0.5 0.85 0.85>, Alpha=$Emitter.AccelerationRamp))
      Front End=&Multiply_Vector(A=<0.3 0.4 0.4>, B=&Lerp_Vector(A=<0.6 0.7 0.7>, B=<0.5 0.85 0.85>, Alpha=$Emitter.AccelerationRamp))
     )
     B=&Multiply_Vector(A=<0.6 1 1>, B=&Lerp_Vector(A=<0.8 1 1>, B=<0.8 0.87 0.87>, Alpha=$Emitter.AccelerationRamp))
     Alpha=HlslExpression() {hlslExpression=Particles.SpawnGroup}
    )
    Module.Write Color=true
    Module.Write Lifetime=true
    Module.Write Mass=false
    Module.Write Material Random=true
    Module.Write Position=false
    Module.Write Scale=true
    Module.Write SpriteRotation=true
    Module.Write SpriteSize=false
    Position=&SimulationPosition()
    Sprite Rotation=&UniformRangedFloat(Minimum=-0.2, Maximum=0.2)
    Sprite Size=<5 5>
   )
  ]
  particleUpdateScript [
   &ParticleState()
   &SkeletalMeshLocation(
    Apply Sampled Orientation to Mesh Particles=true
    Bone / Socket Index=1
    Inherit Velocity=false
    Interpolate Mesh Sampling=false
    MaskSpawnGroup=false
    Mesh Sampling Type=::NewEnumerator5
    Sampled Position Offset=&Lerp_Vector(A=<0 0 -70>, B=<0 0 -75>, Alpha=&MakeFloatFromInt(Int=$Particles.Initial.SpawnGroup))
    Sampled Position Offset Coordinate Space:ENiagaraMeshTransforms=::NewEnumerator3
    Skeletal Mesh=$User.SkeletalMesh
    Socket Sampling Mode=::NewEnumerator5
   )
   &ScaleColor(
    Scale Alpha=&FloatFromCurve(FloatCurve=DataInterface() {propertyValues {Curve {keys [
         <at 0 0 RCIM_Cubic>
         <at 0.3 1>
         {arriveTangent=0.0062, interpMode=RCIM_Cubic, leaveTangent=0.0062, tangentMode=RCTM_User, time=0.5, value=1}
         {arriveTangent=-3.9755, interpMode=RCIM_Cubic, leaveTangent=-3.9755, tangentMode=RCTM_User, time=1, value=0}
        ]}}})
    Scale RGB=<0.75 1 1>
    ScaleA=true
    ScaleRGB=true
   )
   /SetVariables(Particles.BlueScale:Vector3f=&Lerp_Vector(
     A=<0 0 0>
     B=&Multiply_Vector(
      A=&Multiply_Vector(
       A=$Particles.Initial.Scale
       B=&VectorFromCurve(VectorCurve=DataInterface() {propertyValues {
          XCurve {keys [
            {arriveTangent=0.0368, interpMode=RCIM_Cubic, leaveTangent=0.0368, tangentMode=RCTM_User, time=0, value=1.5}
            {arriveTangent=1.3132, interpMode=RCIM_Cubic, leaveTangent=1.3132, tangentMode=RCTM_User, time=1, value=2.2}
           ]}
          YCurve {keys [
            {arriveTangent=0.0463, interpMode=RCIM_Cubic, leaveTangent=0.0463, tangentMode=RCTM_User, time=0, value=0.5}
            {arriveTangent=-0.8757, interpMode=RCIM_Cubic, leaveTangent=-2.7939, tangentMode=RCTM_User, time=0.9977, value=0.1}
           ]}
          ZCurve {keys [<at 0 0.5 RCIM_Cubic>, {arriveTangent=-1.0353, interpMode=RCIM_Cubic, leaveTangent=-2.975, tangentMode=RCTM_User, time=0.9977, value=0.1}]}
         }})
      )
      B=&VectorFromCurve(
       VectorCurve=DataInterface() {propertyValues {
         XCurve {keys [
           <at 0.2506 0.7 RCIM_Cubic>
           {arriveTangent=0.9503, interpMode=RCIM_Cubic, leaveTangent=0.9503, tangentMode=RCTM_User, time=0.9973, value=1.4}
          ]}
         YCurve {keys [<at 0.2506 0.7 RCIM_Cubic>,<at 1 0.7 RCIM_Cubic>]}
         ZCurve {keys [<at 0.2506 0.7 RCIM_Cubic>,<at 1 0.7 RCIM_Cubic>]}
        }}
       CurveIndex=$Emitter.AccelerationRamp
      )
     )
     Alpha=HlslExpression() {hlslExpression="float(Particles.SpawnGroup)"}
    )) {isSetParameters=true, module=SetVariables_#7}
   /SetVariables(Particles.PurpleScale:Vector3f=&Lerp_Vector(
     A=&Multiply_Vector(
      A=&Multiply_Vector(
       A=$Particles.Initial.Scale
       B=&Lerp_Vector(
        A=&VectorFromCurve(VectorCurve=DataInterface() {propertyValues {
           XCurve {keys [
             {arriveTangent=4.2301, interpMode=RCIM_Cubic, time=0, value=3}
             {arriveTangent=0.0169, interpMode=RCIM_Cubic, leaveTangent=0.0169, tangentMode=RCTM_User, time=1, value=4}
            ]}
           YCurve {keys [
             {arriveTangent=0.0463, interpMode=RCIM_Cubic, leaveTangent=0.0463, tangentMode=RCTM_User, time=0, value=0.98}
             {arriveTangent=-1.8909, interpMode=RCIM_Cubic, leaveTangent=-1.8909, tangentMode=RCTM_User, time=1, value=0.5}
            ]}
           ZCurve {keys [<at 0 0.98 RCIM_Cubic>, {arriveTangent=-1.8424, interpMode=RCIM_Cubic, leaveTangent=-1.8424, tangentMode=RCTM_User, time=1, value=0.5}]}
          }})
        B=&VectorFromCurve(VectorCurve=DataInterface() {propertyValues {
           XCurve {keys [
             {arriveTangent=4.2301, interpMode=RCIM_Cubic, leaveTangent=4.2301, tangentMode=RCTM_User, time=0, value=0.2}
             {arriveTangent=0.0169, interpMode=RCIM_Cubic, leaveTangent=0.0169, tangentMode=RCTM_User, time=1, value=6}
            ]}
           YCurve {keys [
             {arriveTangent=0.0463, interpMode=RCIM_Cubic, leaveTangent=0.0463, tangentMode=RCTM_User, time=0, value=0.98}
             {interpMode=RCIM_Cubic, leaveTangent=-1.8909, time=1, value=0.8}
            ]}
           ZCurve {keys [<at 0 0.98 RCIM_Cubic>, {interpMode=RCIM_Cubic, leaveTangent=-1.8424, time=1, value=0.8}]}
          }})
        Alpha=&ClampFloat(Float=$System.AccelerationRamp_02)
       )
      )
      B=<1 1 1>
     )
     B=<0 0 0>
     Alpha=HlslExpression() {hlslExpression="float(Particles.SpawnGroup)"}
    )) {isSetParameters=true, module=SetVariables_#8}
   /SetVariables(Particles.MeshOrientation=&MultiplyQuaternion(Quaternion A=$Particles.MeshOrientation, Quaternion B=<0 1 0 1>)) {isSetParameters=true, module=SetVariables_#9}
   /SetVariables(Particles.MeshOrientation=&MultiplyQuaternion(
     Quaternion A=$Particles.MeshOrientation
     Quaternion B=&MakeQuatFromAxisAngle(Axis=<1 0 0>, AngleInDegrees=&Multiply_Float(A=$Particles.MaterialRandom, B=360))
    )) {isSetParameters=true, module=SetVariables_#10}
   &CameraOffset()
  ]
  rendererClasses [&NiagaraMeshRendererProperties]
  renderers [
   {rendererClass=&NiagaraMeshRendererProperties, rendererIndex=0}
   {rendererClass=&NiagaraMeshRendererProperties, rendererIndex=1}
  ]
 }
 Mini {
  emitterUpdateScript [
   /SetVariables(Emitter.AccelerationRamp:NiagaraFloat=$System.AccelerationRamp) {isSetParameters=true}
   &EmitterState()
   &SpawnRate(SpawnRate=&Lerp_Float(A=45, B=90, Alpha=$Emitter.AccelerationRamp), Use Spawn Probability=false)
  ]
  particleSpawnScript [
   &FN_SetSpawnGroup_v2(Number of Spawn Groups=4) {enabled=false}
   &FN_SetSpawnGroup()
   &InitializeParticle(
    Color=&LerpLinearColors(
     StartColor=&LerpLinearColors(StartColor=<rgba 5 2 5 0.3>, EndColor=<rgba 10 0 1 3>, LerpFactor=&ClampFloat(Float=$System.AccelerationRamp_02))
     EndColor=<rgba 10 0 1 5>
     LerpFactor=&MakeCustomFloatFromBool(Bool=$User.bIsFrontEndPreview, True Float=1, False Float=0)
    )
    Lifetime=&Lerp_Float(A=0.4, B=0.3, Alpha=$Emitter.AccelerationRamp)
    Mesh Scale=&Lerp_Vector(
     A=&MakeVector(X=0.15, Y=0.1, Z=0.1)
     B=<0.4 0.09 0.09>
     Alpha=&ClampFloat(Float=$System.AccelerationRamp_02, Max=0)
    )
    Module.Write Color=true
    Module.Write Lifetime=true
    Module.Write Mass=true
    Module.Write Material Random=true
    Module.Write Position=false
    Module.Write Scale=true
    Module.Write SpriteRotation=true
    Module.Write SpriteSize=false
    Sprite Rotation=&UniformRangedFloat(Minimum=-0.2, Maximum=0.2)
    Sprite Size=<5 5>
   )
  ]
  particleUpdateScript [
   &ParticleState()
   &SkeletalMeshLocation(
    Apply Sampled Orientation to Mesh Particles=true
    Bone / Socket Index=&Add_Integer(A=$Particles.Initial.SpawnGroup, B=4)
    Inherit Velocity=false
    Interpolate Mesh Sampling=false
    MaskSpawnGroup=false
    Mesh Sampling Type=::NewEnumerator5
    Skeletal Mesh=$User.SkeletalMesh
    Socket Sampling Mode=::NewEnumerator5
   ) {module=SkeletalMeshLocation001}
   &ScaleMeshSize(Scale Factor=&MakeVector(
     X=&FloatFromCurve(FloatCurve=DataInterface() {propertyValues {Curve {keys [<at 0 0.8 RCIM_Cubic>, {arriveTangent=2.6748, interpMode=RCIM_Cubic, leaveTangent=2.6748, tangentMode=RCTM_User, time=0.75, value=1.5}]}}})
     Y=&FloatFromCurve(FloatCurve=DataInterface() {propertyValues {Curve {keys [<at 0 1 RCIM_Cubic>,<at 1 0.3>]}}})
     Z=&FloatFromCurve(FloatCurve=DataInterface() {propertyValues {Curve {keys [<at 0 1 RCIM_Cubic>,<at 1 0.3>]}}})
    )) {module=ScaleMeshSize001}
   &ScaleColor(
    Scale Alpha=&FloatFromCurve(FloatCurve=DataInterface() {propertyValues {Curve {keys [
         <at 0 0 RCIM_Cubic>
         {arriveTangent=2, interpMode=RCIM_Cubic, leaveTangent=2, time=0.45, value=1}
         {arriveTangent=-1.8182, interpMode=RCIM_Cubic, leaveTangent=-1.8182, time=0.5, value=1}
         {arriveTangent=0.0656, interpMode=RCIM_Cubic, leaveTangent=0.0656, tangentMode=RCTM_User, time=1, value=0}
        ]}}})
    ScaleA=true
    ScaleRGB=true
   )
   /SetVariables(Particles.MeshOrientation=&MultiplyQuaternion(Quaternion A=$Output.SkeletalMeshLocation001.SampledRotation, Quaternion B=<0 5000 0 1>)) {isSetParameters=true}
   /SetVariables(Particles.MeshOrientation=&MultiplyQuaternion(
     Quaternion A=$Particles.MeshOrientation
     Quaternion B=&MakeQuatFromAxisAngle(Axis=<1 0 0>, AngleInDegrees=&Multiply_Float(A=$Particles.Initial.MaterialRandom, B=360))
    )) {isSetParameters=true}
  ]
  rendererClasses [&NiagaraMeshRendererProperties]
  renderers [{rendererClass=&NiagaraMeshRendererProperties, rendererIndex=0}]
 }
 crystal_shards_L {
  emitterUpdateScript [
   /SetVariables(Emitter.AccelerationRamp:NiagaraFloat=&ClampFloat(Float=$System.AccelerationRamp_02)) {isSetParameters=true, module=SetVariables_#3}
   &EmitterState()
   &SpawnRate(
    Spawn Probability=0.4
    SpawnRate=&Lerp_Float(A=35, B=25, Alpha=$Emitter.AccelerationRamp)
    Use Spawn Probability=true
   )
  ]
  particleSpawnScript [
   &FN_SetSpawnGroup_v2(Number of Spawn Groups=2) {enabled=false}
   &FN_SetSpawnGroup(Number of Spawn Groups=2)
   &InitializeParticle(
    Color=&LerpLinearColors(
     StartColor=&MakeLinearColorFromVectorAndFloat(
      "Vector (RGB)"=&MakeVectorFromLinearColorRGB(Color=<rgba 0 80.9736 100 1>)
      "Float (Alpha)"=&Lerp_Float(
       A=&UniformRangedFloat(Minimum=0.8, Maximum=1.2)
       B=&UniformRangedFloat(Minimum=1.5, Maximum=3)
       Alpha=$Emitter.AccelerationRamp
      )
     )
     EndColor=<rgba 50 0 50 1>
     LerpFactor=HlslExpression() {hlslExpression=Particles.SpawnGroup}
    )
    Lifetime=&Lerp_Float(
     A=&UniformRangedFloat(Minimum=0.9, Maximum=1.35)
     B=&UniformRangedFloat(Minimum=0.3, Maximum=0.6)
     Alpha=HlslExpression() {hlslExpression=Particles.SpawnGroup}
    )
    Mesh Scale=&Lerp_Vector(
     A=&Multiply_Vector(
      A=<0.05 0.05 0.0125>
      B=&Lerp_Vector(
       A=&VectorFromFloat(Value=&UniformRangedFloat(Minimum=0.666, Maximum=1.35))
       B=&UniformRangedVector(Minimum=<5 0.2 0.2>, Maximum=<25 0.8 0.8>)
       Alpha=$Emitter.AccelerationRamp
      )
     )
     B=&Multiply_Vector(
      A=<0.05 0.05 0.0125>
      B=&Lerp_Vector(
       A=&VectorFromFloat(Value=&UniformRangedFloat(Minimum=0.666, Maximum=1.35))
       B=&UniformRangedVector(Minimum=<10 0.2 0.2>, Maximum=<65 0.8 0.8>)
       Alpha=$Emitter.AccelerationRamp
      )
     )
     Alpha=HlslExpression() {hlslExpression=Particles.SpawnGroup}
    )
    Module.Write Color=true
    Module.Write Lifetime=true
    Module.Write Mass=false
    Module.Write Material Random=true
    Module.Write Position=true
    Module.Write Scale=true
    Module.Write SpriteRotation=false
    Module.Write SpriteSize=false
    Position=&Add_Vector(A=&SimulationPosition(), B=<12 90 -35>)
    Sprite Size=<5 5>
   )
   /Niagara/Modules/Spawn/Orientation/V2/InitialMeshOrientation(Use Orientation Vector=true, Use Rotation Vector=false) {enabled=false}
   /Niagara/Modules/Spawn/Orientation/V3/InitialMeshOrientation(Mesh Orientation Mode=::NewEnumerator2, Use Rotation Vector=false) {module=InitialMeshOrientation001}
   &AddVelocity(Velocity=&RotateVectorByQuaternion(
     Vector To Rotate=&MakeVector(
      X=0
      Y=&Lerp_Float(
       A=&UniformRangedFloat(Minimum=90, Maximum=210)
       B=&UniformRangedFloat(Minimum=600, Maximum=900)
       Alpha=$Emitter.AccelerationRamp
      )
      Z=0
     )
     Quaternion=<0 0 0.05 1>
    ))
   &SkeletalMeshLocation(
    Inherit Velocity=false
    MaskSpawnGroup=false
    Mesh Sampling Type=::NewEnumerator5
    Sampled Position Offset Coordinate Space:ENiagaraMeshTransforms=::NewEnumerator3
    Skeletal Mesh=$User.SkeletalMesh
    Socket Sampling Mode=::NewEnumerator5
   )
   &CylinderLocation(
    Cylinder Height=10
    Cylinder Radius=&Lerp_Float(
     A=&UniformRangedFloat(Minimum=0, Maximum=10)
     B=&UniformRangedFloat(Minimum=20, Maximum=25)
     Alpha=HlslExpression() {hlslExpression=Particles.SpawnGroup}
    )
    MaskSpawnGroup=false
    Module.Enable Lathe Profile=false
    Module.Override Randomness=false
    Module.Surface Only=true
    Orientation Axis:ENiagaraOrientationAxis=::YAxis
    Surface Only Band Thickness:NiagaraFloat=&Lerp_Float(A=5, B=2, Alpha=HlslExpression() {hlslExpression=Particles.SpawnGroup})
    Use Endcaps In Surface Only Mode=false
   )
  ]
  particleUpdateScript [
   &ParticleState()
   /Niagara/Modules/Update/Forces/CurlNoiseForce(Noise Frequency=5, Noise Strength=150, Pan Noise=false, Random Seed=1)
   &SolveForcesAndVelocity(Clamp Velocity=false, Limit Acceleration=false)
   &ScaleMeshSize(Scale Factor=&Lerp_Vector(
     A=&VectorFromFloat(Value=&FloatFromCurve(FloatCurve=DataInterface() {propertyValues {Curve {keys [
           <at 0 1>
           {arriveTangent=-0.5431, interpMode=RCIM_Cubic, leaveTangent=-0.5431, tangentMode=RCTM_User, time=0.3, value=1}
           {arriveTangent=-5.5854, interpMode=RCIM_Cubic, leaveTangent=-5.5854, tangentMode=RCTM_User, time=1, value=0}
          ]}}}))
     B=&VectorFromFloat(Value=&FloatFromCurve(FloatCurve=DataInterface() {propertyValues {Curve {keys [
           <at 0 1>
           {arriveTangent=-0.5431, interpMode=RCIM_Cubic, leaveTangent=-0.5431, tangentMode=RCTM_User, time=0.6, value=1}
           {arriveTangent=-5.5854, interpMode=RCIM_Cubic, leaveTangent=-5.5854, tangentMode=RCTM_User, time=1, value=0}
          ]}}}))
     Alpha=HlslExpression() {hlslExpression=Particles.SpawnGroup}
    ))
   &ScaleColor(
    Scale Alpha=&FloatFromCurve(FloatCurve=DataInterface() {propertyValues {Curve {keys [<at 0 1>,<at 1 0>]}}})
    Scale RGB=&VectorFromFloat(Value=&FloatFromCurve(FloatCurve=DataInterface() {propertyValues {Curve {keys [<at 0 1>,<at 1 0>]}}}))
    ScaleA=true
    ScaleRGB=true
   )
  ]
  rendererClasses [&NiagaraMeshRendererProperties]
  renderers [{rendererClass=&NiagaraMeshRendererProperties, rendererIndex=0}]
 }
 crystal_shards_R {
  emitterUpdateScript [
   /SetVariables(Emitter.AccelerationRamp:NiagaraFloat=&ClampFloat(Float=$System.AccelerationRamp_02)) {isSetParameters=true, module=SetVariables_#3}
   &EmitterState()
   &SpawnRate(
    Spawn Probability=0.4
    SpawnRate=&Lerp_Float(A=35, B=25, Alpha=$Emitter.AccelerationRamp)
    Use Spawn Probability=true
   )
  ]
  particleSpawnScript [
   &FN_SetSpawnGroup_v2(Number of Spawn Groups=2) {enabled=false}
   &FN_SetSpawnGroup(Number of Spawn Groups=2)
   &InitializeParticle(
    Color=&LerpLinearColors(
     StartColor=&MakeLinearColorFromVectorAndFloat(
      "Vector (RGB)"=&MakeVectorFromLinearColorRGB(Color=<rgba 0 80.9736 100 1>)
      "Float (Alpha)"=&Lerp_Float(
       A=&UniformRangedFloat(Minimum=0.8, Maximum=1.2)
       B=&UniformRangedFloat(Minimum=1.5, Maximum=3)
       Alpha=$Emitter.AccelerationRamp
      )
     )
     EndColor=<rgba 50 0 50 1>
     LerpFactor=HlslExpression() {hlslExpression=Particles.SpawnGroup}
    )
    Lifetime=&Lerp_Float(
     A=&UniformRangedFloat(Minimum=0.9, Maximum=1.35)
     B=&UniformRangedFloat(Minimum=0.3, Maximum=0.6)
     Alpha=HlslExpression() {hlslExpression=Particles.SpawnGroup}
    )
    Mesh Scale=&Lerp_Vector(
     A=&Multiply_Vector(
      A=<0.05 0.05 0.0125>
      B=&Lerp_Vector(
       A=&VectorFromFloat(Value=&UniformRangedFloat(Minimum=0.666, Maximum=1.35))
       B=&UniformRangedVector(Minimum=<5 0.2 0.2>, Maximum=<25 0.8 0.8>)
       Alpha=$Emitter.AccelerationRamp
      )
     )
     B=&Multiply_Vector(
      A=<0.05 0.05 0.0125>
      B=&Lerp_Vector(
       A=&VectorFromFloat(Value=&UniformRangedFloat(Minimum=0.666, Maximum=1.35))
       B=&UniformRangedVector(Minimum=<10 0.2 0.2>, Maximum=<65 0.8 0.8>)
       Alpha=$Emitter.AccelerationRamp
      )
     )
     Alpha=HlslExpression() {hlslExpression=Particles.SpawnGroup}
    )
    Module.Write Color=true
    Module.Write Lifetime=true
    Module.Write Mass=false
    Module.Write Material Random=true
    Module.Write Position=true
    Module.Write Scale=true
    Module.Write SpriteRotation=false
    Module.Write SpriteSize=false
    Position=&Add_Vector(A=&SimulationPosition(), B=<12 90 -35>)
    Sprite Size=<5 5>
   )
   /Niagara/Modules/Spawn/Orientation/V2/InitialMeshOrientation(Use Orientation Vector=true, Use Rotation Vector=false) {enabled=false}
   /Niagara/Modules/Spawn/Orientation/V3/InitialMeshOrientation(Mesh Orientation Mode=::NewEnumerator2, Use Rotation Vector=false) {module=InitialMeshOrientation001}
   &AddVelocity(Velocity=&RotateVectorByQuaternion(
     Vector To Rotate=&MakeVector(
      X=0
      Y=&Lerp_Float(
       A=&UniformRangedFloat(Minimum=90, Maximum=210)
       B=&UniformRangedFloat(Minimum=600, Maximum=900)
       Alpha=$Emitter.AccelerationRamp
      )
      Z=0
     )
     Quaternion=<0 0 -0.05 1>
    ))
   &SkeletalMeshLocation(
    Bone / Socket Index=1
    Inherit Velocity=false
    MaskSpawnGroup=false
    Mesh Sampling Type=::NewEnumerator5
    Sampled Position Offset Coordinate Space:ENiagaraMeshTransforms=::NewEnumerator3
    Skeletal Mesh=$User.SkeletalMesh
    Socket Sampling Mode=::NewEnumerator5
   )
   &CylinderLocation(
    Cylinder Height=10
    Cylinder Radius=&Lerp_Float(
     A=&UniformRangedFloat(Minimum=0, Maximum=10)
     B=&UniformRangedFloat(Minimum=20, Maximum=25)
     Alpha=HlslExpression() {hlslExpression=Particles.SpawnGroup}
    )
    MaskSpawnGroup=false
    Module.Enable Lathe Profile=false
    Module.Override Randomness=false
    Module.Surface Only=true
    Orientation Axis:ENiagaraOrientationAxis=::YAxis
    Surface Only Band Thickness:NiagaraFloat=&Lerp_Float(A=5, B=2, Alpha=HlslExpression() {hlslExpression=Particles.SpawnGroup})
    Use Endcaps In Surface Only Mode=false
   )
  ]
  particleUpdateScript [
   &ParticleState()
   &CurlNoiseForce(Noise Frequency=5, Noise Strength=150, Pan Noise=false, Random Seed=1)
   &SolveForcesAndVelocity(Clamp Velocity=false, Limit Acceleration=false)
   &ScaleMeshSize(Scale Factor=&Lerp_Vector(
     A=&VectorFromFloat(Value=&FloatFromCurve(FloatCurve=DataInterface() {propertyValues {Curve {keys [
           <at 0 1>
           {arriveTangent=-0.5431, interpMode=RCIM_Cubic, leaveTangent=-0.5431, tangentMode=RCTM_User, time=0.3, value=1}
           {arriveTangent=-5.5854, interpMode=RCIM_Cubic, leaveTangent=-5.5854, tangentMode=RCTM_User, time=1, value=0}
          ]}}}))
     B=&VectorFromFloat(Value=&FloatFromCurve(FloatCurve=DataInterface() {propertyValues {Curve {keys [
           <at 0 1>
           {arriveTangent=-0.5431, interpMode=RCIM_Cubic, leaveTangent=-0.5431, tangentMode=RCTM_User, time=0.6, value=1}
           {arriveTangent=-5.5854, interpMode=RCIM_Cubic, leaveTangent=-5.5854, tangentMode=RCTM_User, time=1, value=0}
          ]}}}))
     Alpha=HlslExpression() {hlslExpression=Particles.SpawnGroup}
    ))
   &ScaleColor(
    Scale Alpha=&FloatFromCurve(FloatCurve=DataInterface() {propertyValues {Curve {keys [<at 0 1>,<at 1 0>]}}})
    Scale RGB=&VectorFromFloat(Value=&FloatFromCurve(FloatCurve=DataInterface() {propertyValues {Curve {keys [<at 0 1>,<at 1 0>]}}}))
    ScaleA=true
    ScaleRGB=true
   )
  ]
  rendererClasses [&NiagaraMeshRendererProperties]
  renderers [{rendererClass=&NiagaraMeshRendererProperties, rendererIndex=0}]
 }
}
summary {
 systemName=NS_Glider_SpaceWanderer_Male_Thruster
 emitters [
  {emitterName=Front_Flare_Large_L, rendererClasses [&NiagaraMeshRendererProperties]}
  {emitterName=Blue_Decal_Large, rendererClasses [&NiagaraSpriteRendererProperties]}
  {emitterName=Mini, rendererClasses [&NiagaraMeshRendererProperties]}
  {emitterName=crystal_shards_L, rendererClasses [&NiagaraMeshRendererProperties]}
  {emitterName=Main_L, rendererClasses [&NiagaraMeshRendererProperties]}
  {emitterName=Fire_R, rendererClasses [&NiagaraSpriteRendererProperties]}
  {emitterName=crystal_shards_R, rendererClasses [&NiagaraMeshRendererProperties]}
  {emitterName=Front_Flare_Large_R, rendererClasses [&NiagaraMeshRendererProperties]}
  {emitterName=Light, rendererClasses [&NiagaraLightRendererProperties]}
  {emitterName=Fire_L, rendererClasses [&NiagaraSpriteRendererProperties]}
  {emitterName=Blue_Decal_Small, rendererClasses [&NiagaraSpriteRendererProperties]}
  {emitterName=Main_R, rendererClasses [&NiagaraMeshRendererProperties]}
 ]
 userVariables [
  User.bIsFrontEnd {type=NiagaraBool}
  User.bIsFrontEndPreview {type=NiagaraBool}
  User.FEVec {type=Vector3f, description="", defaultValue {struct=Vector3f, value=<0 0 0>}}
  User.ForwardDot {type=NiagaraFloat, defaultValue=1, description=""}
  User.Open {type=NiagaraFloat, defaultValue=0.15, description=""}
  User.PlayerSpeed {type=NiagaraFloat, defaultValue=500, description=""}
  User.RibbonIDOffset {type=NiagaraInt32}
  User.SizeScale {type=NiagaraFloat, defaultValue=0, description=""}
  User.SkeletalMesh {
   type=NiagaraDataInterfaceSkeletalMesh
   defaultValue {
    struct=NiagaraExt_VariableValue_DataInterface
    value {dataInterface=&self:NiagaraDataInterfaceSkeletalMesh_2}
   }
  }
 ]
}
